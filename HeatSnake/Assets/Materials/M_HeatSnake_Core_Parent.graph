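# Unlit, masked, two-sided parent material for the glowing core of the skeletal HeatSnake, wrapped in the shared CharacterParent PreFX and PostFX functions.
# A time sine running along UV V pulses the hot orange emissive and the normal-based world position offset.
# Glow strength adds parameter-collection opacity to a squared fade timed from BP_GetTime.
# Named reroutes share masks from pre-skinned height, UV caps and Fresnel.
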Material /HeatSnake/Assets/Materials/M_HeatSnake_Core_Parent
graph {
 properties {
  bAutomaticallySetUsageInEditor=false
  bUseMaterialAttributes=true
  bUsedWithMeshDeformer=true
  bUsedWithMorphTargets=true
  bUsedWithSkeletalMesh=true
  blendMode=BLEND_Masked
  pixelDepthOffsetMode=PDOM_Legacy
  shadingModel=MSM_Unlit
  twoSided=true
  usedShadingModels=Unlit
 }
 nodes [
  n1 = MaterialFunctionCall() {materialFunction=&MF_CharacterParent_2023_PreFX}
  n2 = Multiply(A=Time()) {constB=0.5}
  n3 = Multiply(A=ComponentMask(None=TextureCoordinate()) {g=true})
  n4 = ConstantBiasScale(None=Sine(None=Add(A=@n2, B=@n3)))
  n5 = LinearInterpolate(Alpha=@n4) {constA=-0.65, constB=-0.35}
  n6 = LinearInterpolate(Alpha=NamedRerouteUsage()) {constA=-0.9, constB=0}
  n7 = NamedRerouteUsage()
  n8 = Constant3Vector() {constant=<rgba 3 0.151 0 1>}
  n9 = OneMinus(None=NamedRerouteUsage())
  n10 = Multiply(
   A=LinearInterpolate(A=Multiply(A=@n8) {constB=5}, B=Multiply(A=@n8) {constB=14}, Alpha=Power(Base=@n9))
   B=LinearInterpolate(Alpha=@n4) {constA=0.8}
  )
  n11 = Subtract(A=Time(), B=CollectionParameter() {collection=&/MPC_HeatSnake, parameterName=BP_GetTime})
  n12 = OneMinus(None=Saturate(None=Multiply(A=@n11) {constB=0.7}))
  n13 = NamedRerouteDeclaration(None=Add(A=CollectionParameter() {collection=&MPC_HeatSnake, parameterName=BP_GlowOpacity}, B=Multiply(A=@n12, B=@n12))) {name=CollectionParameter_GlowOpacity, nodeColor=<rgba 0.1719 0 1 1>}
  n14 = SetMaterialAttributes!(
   MaterialAttributes=@n1.Result
   Normal
   World Position Offset=Multiply(A=VertexNormalWS(), B=LinearInterpolate(A=@n5, B=@n6, Alpha=@n7) {constB=0}) {constB=-10}
   Opacity Mask=NamedRerouteUsage()
   Emissive Color=Multiply(A=Multiply(A=@n10) {constB=2}, B=LinearInterpolate(Alpha=@n13) {constA=0.1, constB=2.5})
   Base Color
   Metallic
   Specular
   Roughness
  ) {attributeSetTypes [
    #1
    #2
    #3
    #4
    #5
    #6
    #7
    #8
   ]}
  MaterialFunctionCall(Material=@n14, M=@n1.M, Pre Skinned Local Position=@n1.M, Distance Blend=@n1.M, Fresnel=@n1.M) {materialFunction=&MF_CharacterParent_2023_PostFX}
  NamedRerouteDeclaration(None=@n1.Pre skinned Local Position) {name=PS_LocalPosition, nodeColor=<rgba 0 1 0.0859 1>}
  n15 = ComponentMask(None=NamedRerouteUsage()) {b=true}
  n16 = Saturate(None=Divide(A=Add(A=@n15) {constB=-68}) {constB=40})
  NamedRerouteDeclaration(None=@n16) {name=PointThicknessMask, nodeColor=<rgba 1 0 0.7969 1>}
  n17 = ComponentMask(None=TextureCoordinate()) {g=true}
  n18 = Step(X=@n17) {constY=0.3125}
  n19 = Step(X=@n17) {constY=0.6875}
  NamedRerouteDeclaration(None=Saturate(None=Subtract(A=@n18, B=@n19))) {name=CapsMask, nodeColor=<rgba 1 0 0.1641 1>}
  NamedRerouteDeclaration(None=@n1.Fresnel) {name=InFresnel, nodeColor=<rgba 0 0.1094 1 1>}
  OneMinus(None=@n16)
 ]
}
