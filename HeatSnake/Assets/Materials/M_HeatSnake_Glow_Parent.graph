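# Additive, two-sided parent material for the HeatSnake character's glow on skeletal meshes, built between the shared character PreFX and PostFX functions.
# An orange fresnel gradient, banded by UV and oversaturated, forms the emissive, scaled by glow opacity from the HeatSnake parameter collection.
# World position offset pushes along normals with a time-and-UV sine pulse and pre-skinned position masks.
# An attack-timed opacity fade is left unconnected.
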Material /HeatSnake/Assets/Materials/M_HeatSnake_Glow_Parent
graph {
 properties {
  bAutomaticallySetUsageInEditor=false
  bUseMaterialAttributes=true
  bUsedWithMeshDeformer=true
  bUsedWithMorphTargets=true
  bUsedWithSkeletalMesh=true
  blendMode=BLEND_Additive
  pixelDepthOffsetMode=PDOM_Legacy
  shadingModel=MSM_FromMaterialExpression
  twoSided=true
  usedShadingModels=Default Lit | Subsurface | Clear Coat | Subsurface Profile | Cloth
 }
 nodes [
  Constant3Vector() {constant=<rgba 0 0 1 0>}
  n1 = MaterialFunctionCall() {materialFunction=&MF_CharacterParent_2023_PreFX}
  n2 = CollectionParameter() {collection=&/MPC_HeatSnake, parameterName=BP_GlowOpacity}
  n3 = SetMaterialAttributes!(
   MaterialAttributes=@n1.Result
   Normal
   World Position Offset=NamedRerouteUsage()
   Opacity Mask
   Emissive Color=Multiply(A=NamedRerouteUsage(), B=@n2)
   Base Color
  ) {attributeSetTypes [
    #1
    #2
    #3
    #4
    #5
   ]}
  MaterialFunctionCall(Material=@n3, M=@n1.M, Pre Skinned Local Position=@n1.M, Distance Blend=@n1.M, Fresnel=@n1.M) {materialFunction=&MF_CharacterParent_2023_PostFX}
  n4 = Subtract(A=Subtract(A=Time()), B=Constant() {desc=3-attackCombo-Duration, r=1.4}) {constB=1.2}
  n5 = Saturate(None=Multiply(A=@n4) {constB=10})
  n6 = Multiply(A=Subtract(A=@n4) {constB=0.5}) {constB=10}
  Multiply(A=@n2, B=Add(A=OneMinus(None=@n5), B=Saturate(None=@n6)))
  NamedRerouteDeclaration(None=@n1.Pre skinned Local Position) {name=PS_LocalPosition, nodeColor=<rgba 0.9688 0 1 1>}
  n7 = ComponentMask(None=NamedRerouteUsage()) {b=true}
  n8 = Divide(A=Add(A=@n7) {constB=-68}) {constB=40}
  n9 = Saturate(None=Divide(A=Add(A=@n7) {constB=-15}) {constB=20})
  NamedRerouteDeclaration(None=Add(A=Saturate(None=@n8), B=OneMinus(None=@n9))) {name=PointThicknessMask, nodeColor=<rgba 0.4766 0 1 1>}
  n10 = Multiply(A=Time()) {constB=0.5}
  n11 = Multiply(A=ComponentMask(None=TextureCoordinate()) {g=true})
  n12 = ConstantBiasScale(None=Sine(None=Add(A=@n10, B=@n11)))
  n13 = Constant() {r=5}
  n14 = LinearInterpolate(
   A=LinearInterpolate(Alpha=@n12) {constA=-0.65, constB=-0.15}
   B=Multiply(A=Subtract(A=@n13) {constB=0.5}) {constB=0}
   Alpha=NamedRerouteUsage()
  ) {constB=-4.8}
  NamedRerouteDeclaration(None=Multiply(A=VertexNormalWS(), B=Add(A=@n14, B=@n13) {constB=10}) {constB=-10}) {name=WPO, nodeColor=<rgba 1 0 0.4922 1>}
  n15 = Constant3Vector() {constant=<rgba 3 1.0896 0.0197 1>}
  n16 = NamedRerouteDeclaration(None=@n1.Fresnel) {name=InFresnel, nodeColor=<rgba 1 0.7969 0 1>}
  n17 = LinearInterpolate(A=Multiply(A=@n15) {constB=2}, B=Multiply(A=@n15) {constB=7}, Alpha=Power(Base=OneMinus(None=@n16)) {constExponent=3})
  n18 = LinearInterpolate(Alpha=@n12) {constA=0.7}
  n19 = OneMinus(None=NamedRerouteUsage())
  n20 = Multiply(A=@n19, B=@n19)
  n21 = Multiply(A=@n20, B=@n20)
  n22 = ComponentMask(None=TextureCoordinate()) {g=true}
  n23 = Step(X=@n22) {constY=0.3125}
  n24 = Step(X=@n22) {constY=0.6875}
  n25 = Multiply(A=Multiply(A=Multiply(A=@n17, B=@n18), B=Add(A=@n21) {constB=0.2}), B=Saturate(None=Subtract(A=@n23, B=@n24)))
  NamedRerouteDeclaration(None=Desaturation(None=Multiply(A=@n25) {constB=0.03}, Fraction=Constant() {r=-1})) {name=OUT_Emissive, nodeColor=<rgba 0 0.9062 1 1>}
  n26 = NamedRerouteUsage()
  n27 = Constant() {r=1}
  n28 = MaterialFunctionCall(R=Sign(None=ComponentMask(None=@n26) {r=true}), G=@n27, B=@n27) {materialFunction=&AppendMany}
  n29 = Multiply(A=Constant3Vector() {constant=<rgba 119.6 54.3 27.4 0>}, B=@n28.RGB)
  n30 = Constant3Vector() {constant=<rgba 55.6 7.6 96 0>}
  n31 = Subtract(A=LocalPosition().XYZ, B=NamedRerouteUsage())
  n32 = Subtract(A=@n29, B=Add(A=Multiply(A=@n30, B=@n28.RGB), B=@n31))
  n33 = Transform(None=Normalize(VectorInput=@n32)) {transformSourceType=TRANSFORMSOURCE_Local}
  NamedRerouteDeclaration(None=MaterialFunctionCall(Axis=Normalize(VectorInput=@n33), Exponent=Constant() {r=4}) {materialFunction=&AxisAlignedFresnel}.Fresnel) {name=AxisFresnel, nodeColor=<rgba 1 0 0.3047 1>}
  n34 = Subtract(A=NamedRerouteUsage(), B=TransformPosition(None=CameraPositionWS()) {transformSourceType=TRANSFORMPOSSOURCE_World})
  n35 = Add(A=LocalPosition().XYZ, B=@n32)
  Normalize(VectorInput=CrossProduct(A=Normalize(VectorInput=@n34), B=Normalize(VectorInput=@n35)))
  ScalarParameter() {defaultValue=0.2, parameterName=FresnelInfluence}
  Subtract(A=@n32, B=@n31)
  Subtract(A=@n29, B=@n31)
  n36 = NamedRerouteUsage()
  n37 = Constant() {r=1}
  n38 = MaterialFunctionCall(R=Sign(None=ComponentMask(None=@n36) {r=true}), G=@n37, B=@n37) {materialFunction=&AppendMany}
  Subtract(
   A=Multiply(A=Constant3Vector() {constant=<rgba 119.6 54.3 27.4 0>}, B=@n38.RGB)
   B=Multiply(A=Constant3Vector() {constant=<rgba 55.6 7.6 96 0>}, B=@n38.RGB)
  )
  WorldPosition()
 ]
}
