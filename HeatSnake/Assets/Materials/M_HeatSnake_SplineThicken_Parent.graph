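# Parent material for HeatSnake skeletal meshes, drawing spline-like strands with a glowing orange, heat-like emissive.
# SplineThicken gives them constant width through World Position Offset and supplies a world normal converted to tangent space.
# A squared triangular falloff on U makes the emissive brightest mid-range.
# These override the shared character PreFX attributes before the PostFX function; Fresnel, camera-facing and WireEndOpacity branches appear unused.
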Material /HeatSnake/Assets/Materials/M_HeatSnake_SplineThicken_Parent
graph {
 properties {
  bAutomaticallySetUsageInEditor=false
  bUseMaterialAttributes=true
  bUsedWithMeshDeformer=true
  bUsedWithMorphTargets=true
  bUsedWithSkeletalMesh=true
  pixelDepthOffsetMode=PDOM_Legacy
  shadingModel=MSM_FromMaterialExpression
  usedShadingModels=Default Lit | Subsurface | Clear Coat | Subsurface Profile | Cloth
 }
 nodes [
  n1 = TextureCoordinate()
  AppendVector(A=OneMinus(None=ComponentMask(None=@n1) {r=true}), B=ComponentMask(None=@n1) {g=true})
  n2 = Constant() {r=5}
  n3 = TextureCoordinate()
  n4 = MaterialFunctionCall(WidthBase=@n2, WidthTip=@n2, UVs for Projection=@n3, UVs for Thickness=@n3, UVs for texturing=@n3, UVs For NormalShading=@n3) {materialFunction=&SplineThicken}
  Fresnel(Normal=@n4.Normal)
  n5 = MaterialFunctionCall() {materialFunction=&MF_CharacterParent_2023_PreFX}
  n6 = Constant3Vector() {constant=<rgba 3 0.1486 0.0046 1>}
  n7 = ComponentMask(None=TextureCoordinate()) {r=true}
  n8 = OneMinus(None=Abs(None=ConstantBiasScale(None=@n7) {bias=-0.5, scale=2}))
  n9 = Multiply(A=@n8, B=@n8)
  n10 = SetMaterialAttributes!(
   MaterialAttributes=@n5.Result
   Normal=Transform(None=@n4.Normal) {transformSourceType=TRANSFORMSOURCE_World, transformType=TRANSFORM_Tangent}
   World Position Offset=@n4.WorldPositionOffset
   Opacity Mask
   Emissive Color=LinearInterpolate(A=Multiply(A=@n6) {constB=7}, B=@n6, Alpha=OneMinus(None=@n9))
  ) {attributeSetTypes [
    #1
    #2
    #3
    #4
   ]}
  MaterialFunctionCall(Material=@n10, M=@n5.M, Pre Skinned Local Position=@n5.M, Distance Blend=@n5.M, Fresnel=@n5.M) {materialFunction=&MF_CharacterParent_2023_PostFX}
  n11 = ConstantBiasScale(None=DotProduct(A=@n4.Normal, B=MaterialFunctionCall() {materialFunction=&CameraDirectionVector}.Result))
  Multiply(A=@n11, B=@n11)
  n12 = ConstantBiasScale(None=ComponentMask(None=TextureCoordinate()) {g=true}) {bias=-0.5, scale=2}
  NamedRerouteDeclaration(None=Step(Y=Abs(None=@n12)) {constX=0.5, constY=0.5}) {name=WireEndOpacity, nodeColor=<rgba 0.9219 1 0 1>}
  Power(Base=@n11, Exp=ScalarParameter() {defaultValue=2, parameterName=ASFSD}) {constExponent=0.2}
 ]
}
