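# Placeholder surface material for the HeatSnake, using Default Lit shading and flagged for skeletal mesh use.
# It exposes a pink-red Base Color vector parameter and scalar parameters for Metallic, Specular and Roughness.
# The graph has only these constant parameters, with no textures or math nodes, and automatic usage flagging in the editor is off.
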Material /HeatSnake/Assets/m_PLACEHOLDER_HeatSnake
graph {
 properties {bAutomaticallySetUsageInEditor=false, bUsedWithSkeletalMesh=true, usedShadingModels=Default Lit}
 nodes [
  ScalarParameter() {defaultValue=0.3, parameterName=Metallic}
  ScalarParameter() {defaultValue=0.15, parameterName=Specular}
  ScalarParameter() {defaultValue=0.7, parameterName=Roughness}
  VectorParameter() {parameterName=Base Color, defaultValue=<rgba 1 0.032 0.1533 1>}
 ]
}
