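# Placeholder material for the HeatSnake weapon hide, set up as an unlit, alpha-composite material for skeletal meshes.
# It declares Opacity, Base Color (defaulting to dark green) and Emissive Color parameters, with no connections between them or to material outputs.
# Automatic usage setting in the editor is off, leaving skeletal mesh as the only enabled usage shown.
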Material /HeatSnake/Assets/m_PLACEHOLDER_HeatSnakeWeaponHide
graph {
 properties {
  bAutomaticallySetUsageInEditor=false
  bUsedWithSkeletalMesh=true
  blendMode=BLEND_AlphaComposite
  shadingModel=MSM_Unlit
  usedShadingModels=Unlit
 }
 nodes [
  ScalarParameter() {parameterName=Opacity}
  VectorParameter() {parameterName=Base Color, defaultValue=<rgba 0.0106 0.0781 0 1>}
  VectorParameter() {parameterName=Emissive Color}
 ]
}
